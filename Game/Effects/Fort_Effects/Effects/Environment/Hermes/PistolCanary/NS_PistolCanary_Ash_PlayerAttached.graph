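# Niagara system of drifting ash sprites for the PistolCanary environment effects, named for attachment to the player.
# One GPU emitter continuously spawns small, randomly tinted particles throughout a large sphere, launched in a diagonal downward cone.
# Curl noise and vortex forces apply only at spawn, then drag slows particles while alpha fades in and out.
# A rapidly rotating random vector feeds dynamic material parameters for glints, and camera-based scaling keeps tiny flakes visible.
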NiagaraSystem /Game/Effects/Fort_Effects/Effects/Environment/Hermes/PistolCanary/NS_PistolCanary_Ash_PlayerAttached
dependencies {usedDataInterfaces [&NiagaraDataInterfaceCurve], usedDynamicInputs [], usedModules [&SystemState]}
emitters {Ash {
  simTarget=GPUComputeSim
  emitterUpdateScript [
   &EmitterState()
   &SpawnRate(SpawnRate=3000, Use Spawn Probability=false) {module=SpawnRate001}
  ]
  particleSpawnScript [
   &InitializeParticle(
    Color Maximum=<rgba 0.15 0.0389 0.0158 1>
    Color Minimum=<rgba 0.0435 0.0178 0.0076 1>
    Color Mode=::NewEnumerator2
    Lifetime Max=4
    Lifetime Min=3
    Lifetime Mode=::NewEnumerator1
    Sprite Rotation Mode=::NewEnumerator2
    Sprite Size Mode=::NewEnumerator3
    Uniform Sprite Size=&Multiply_Float(A=&RandomRangeFloat(Minimum=0.5, Maximum=1), B=1.3)
    UsePositionOffset=false
    Write Material Random=true
   )
   /Niagara/Modules/Spawn/Location/SphereLocation(MaskSpawnGroup=false, Module.Override Randomness=false, Sphere Radius=5000, Surface Only=false, Surface Only Band Thickness=1)
   &AddVelocityInCone(Cone Angle=1, Cone Axis=<1 0 -1>, Cone Axis Coordinate Space=::World, Module.Use Velocity Falloff On Cone Axis=false, Velocity Strength=150)
   /Niagara/Modules/Update/Forces/CurlNoiseForce(
    Curl Noise Cone Mask Angle=120
    Curl Noise Cone Mask Falloff Angle=100
    Mask Curl Noise=true
    Noise Frequency=250
    Noise Strength=900
    Pan Noise=true
    Pan Noise Field=<0 1 2>
   )
   &VortexForce(Origin Pull Amount=300, UseOriginOffset=true, Vortex Axis=<1 0 0>, Vortex Force Amount=300, Vortex Origin Offset:Vector3f=<0 0 -5000>)
   &ApplyInitialForces(Apply Force to Position=true)
  ]
  particleUpdateScript [
   &ParticleState() {module=UpdateAge}
   &ScaleColor(
    Scale Alpha=&FloatFromCurve(
     FloatCurve=DataInterface() {propertyValues {Curve {keys [
         <at -0.0007 -0.0138 RCIM_Cubic>
         {arriveTangent=0.0948, interpMode=RCIM_Cubic, leaveTangent=0.0948, tangentMode=RCTM_User, time=0.2, value=1}
         {arriveTangent=-0.0232, interpMode=RCIM_Cubic, leaveTangent=-0.0232, tangentMode=RCTM_User, time=0.5424, value=1}
         <at 1.0011 0.002 RCIM_Cubic>
        ]}}}
     Scale Curve=0.6
    )
    ScaleA=true
    ScaleRGB=true
   )
   &AerodynamicDrag(Aerodynamic Drag=0.025, Aerodynamic Rotational Drag=0.1, Use Linear Drag=true, Use Rotational Drag=true)
   &SolveForcesAndVelocity(Clamp Velocity=false, Limit Acceleration=false)
   /Niagara/Modules/Update/Camera/MaintainInCameraParticleScale(Render Target Width=0.1)
   /SetVariables(Particles.RandomGlintVector:Vector3f=&RotateVectorByQuaternion(
     Vector To Rotate=&RandomVector(Vector Scale=1)
     Quaternion=&MakeQuatFromAxisAngle(
      Axis=&RandomVector(Vector Scale=1)
      AngleInDegrees=&Multiply_Float(A=$Particles.NormalizedAge, B=&RandomRangeFloat(Minimum=2000, Maximum=3000))
     )
    )) {isSetParameters=true}
   &DynamicMaterialParameters(
    Index 0 Param 1=&RandomRangeFloat(Minimum=0, Maximum=2)
    Index 0 Param 2=&MakeFloatFromVector(Vector=$Particles.RandomGlintVector, Channel=::NewEnumerator0)
    Index 0 Param 3=&MakeFloatFromVector(Vector=$Particles.RandomGlintVector, Channel=::NewEnumerator1)
    Index 0 Param 4=&MakeFloatFromVector(Vector=$Particles.RandomGlintVector, Channel=::NewEnumerator2)
    Module.Param0WriteEnabled=true
    Module.Param1WriteEnabled=true
    Module.Param2WriteEnabled=true
    Module.Param3WriteEnabled=true
   )
   &AlignSpriteToMeshOrientation(SetAlignment=true, SetFacing=true)
   &ScaleColor(Scale Alpha=10, Scale RGB=<0 1 1>, ScaleA=true, ScaleRGB=true) {enabled=false, module=ScaleColor001}
   &ScaleSpriteSize(Scale Sprite Size Mode=::NewEnumerator0, Uniform Scale Factor=50) {enabled=false}
  ]
  rendererClasses [&NiagaraSpriteRendererProperties]
  renderers [{rendererClass=&NiagaraSpriteRendererProperties}]
 }}
summary {
 systemName=NS_PistolCanary_Ash_PlayerAttached
 emitters [{emitterName=Ash, simTarget=GPUComputeSim, rendererClasses [&NiagaraSpriteRendererProperties]}]
 userVariables []
}
