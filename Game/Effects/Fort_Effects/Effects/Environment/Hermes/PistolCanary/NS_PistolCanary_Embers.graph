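# Niagara environment effect for PistolCanary that fills the area with drifting, flickering embers on the GPU.
# Embers_Parent scatters sprites in a large sphere, snaps them onto the depth buffer and culls invalid samples, while Embers spawns smaller clusters from those particles.
# Both use point force, wind and drag, with a region mask driving speed limit, color and alpha.
# View mask and camera distance scale sprite size, and emitter distance scales spawning.
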NiagaraSystem /Game/Effects/Fort_Effects/Effects/Environment/Hermes/PistolCanary/NS_PistolCanary_Embers
dependencies {
 usedDataInterfaces [
  &NiagaraDataInterfaceCurve
  &NiagaraDataInterfaceCollisionQuery
  &NiagaraDataInterfaceLandscape
  &NiagaraDataInterfaceColorCurve
  &NiagaraDataInterfaceParticleRead
 ]
 usedDynamicInputs []
 usedModules [
  &SystemState
  &self:SystemSpawnScript.SystemScriptSource.SystemScriptGraph.NiagaraNodeAssignment_0.SetVariables_#1
  &self:SystemSpawnScript.SystemScriptSource.SystemScriptGraph.NiagaraNodeAssignment_1.SetVariables_#2
  &self:SystemSpawnScript.SystemScriptSource.SystemScriptGraph.NiagaraNodeAssignment_2.SetVariables_#3
  &self:SystemSpawnScript.SystemScriptSource.SystemScriptGraph.NiagaraNodeAssignment_3.SetVariables_#4
 ]
}
emitters {
 Embers {
  simTarget=GPUComputeSim
  emitterUpdateScript [
   &EmitterState(
    Enable Distance Culling=true
    MaxDistance=18750
    Module.bApplySpawnCountScaleByDistance=true
    Module.bUseMaxDistance=true
    Module.bUseMinDistance=false
    Scalability Mode=::NewEnumerator1
    Scale Spawn Count=true
    Spawn Count Scale By Distance=DataInterface() {propertyValues {Curve {keys [<at 0 0>,<at 0.1 0>,<at 0.5 1>,<at 0.75 1>,<at 1 1>]}}}
   )
   &SpawnParticlesFromOtherEmitter(
    Attribute Reader=DataInterface() {propertyValues {EmitterBinding {bindingMode=Other, emitterName=Embers_Parent}}}
    Spawn Rate=62.5
    Spawn Rate Per Particle Cap=50
    Use Maximum Age=false
    Use Minimum Age=false
   )
  ]
  particleSpawnScript [
   &InitializeParticle(
    Lifetime Max=3
    Lifetime Min=2
    Lifetime Mode=::NewEnumerator1
    Sprite Size Mode=::NewEnumerator1
    Uniform Sprite Size Max=6.5
    Uniform Sprite Size Min=3.5
    UsePositionOffset=false
    Write Material Random=true
   ) {module=InitializeParticle001}
   &CameraOffset(Camera Offset Amount=50)
   /SetVariables(Particles.InitialPos:NiagaraPosition=$Particles.Position) {isSetParameters=true, module=SetVariables_#5}
   /Niagara/Modules/Update/Camera/ViewMask(Border Percent Max=25, Border Percent Min=0) {module=ViewMask001}
   &SampleParticlesFromOtherEmitter(Velocity Application:ENiagara_DirectReadApplicationMode=::NewEnumerator0, Velocity Sampling=::NewEnumerator0)
   &ShapeLocation(
    MaskSpawnGroup=false
    Non Uniform Scale=&RandomRangeVector(Minimum=<0 0 0>, Maximum=<1 1 2>)
    Override Seed=false
    Sphere Radius=&RandomRangeFloat(Minimum=1, Maximum=25)
    UseOwnerScale=true
   )
  ]
  particleUpdateScript [
   &ParticleState()
   &KillParticlesInVolume(Sphere Radius=2500)
   &RegionMask(Region Falloff=1, Region Radius=8000, UseOffset=false, UseRotation=false)
   &PointForce(Force Falloff Distance=250000, Force Strength=-250, Module.Use Falloff=true, Module.UseLinearForce=false)
   &WindForce(
    Alignment Falloff Start / End:Vector2f=<0 250>
    Frequency:NiagaraFloat=&RandomRangeFloat(Minimum=0.1, Maximum=0.2)
    Scale=1
    Wind Speed=<1 0 1>
    Wind Speed Scale=25
   )
   &AerodynamicDrag(
    Aerodynamic Drag:NiagaraFloat=&RandomRangeFloat(Minimum=0.01, Maximum=0.02)
    Lift Contribution=0
    Use Linear Drag=true
    Use Rotational Drag=true
   )
   &SolveForcesAndVelocity(
    Clamp Velocity=true
    Limit Acceleration=false
    Speed Limit:NiagaraFloat=&Multiply_Float(
     A=250
     B=&FloatFromCurve(
      FloatCurve=DataInterface() {propertyValues {Curve {keys [
          <at 0 0.5>
          {arriveTangent=1.255, interpMode=RCIM_Cubic, leaveTangent=1.255, tangentMode=RCTM_User, time=0.5, value=0.5}
          {arriveTangent=2.5018, interpMode=RCIM_Cubic, leaveTangent=2.5018, tangentMode=RCTM_User, time=1, value=2}
         ]}}}
      CurveIndex=$Particles.RegionMask.Float
     )
    )
   )
   &ViewMask(Border Percent Max=15)
   &ScaleColor(
    Scale Alpha=&Multiply_Float(
     A=&Multiply_Float(
      A=&FloatFromCurve(
       FloatCurve=DataInterface() {propertyValues {Curve {keys [
           {interpMode=RCIM_Cubic, tangentMode=RCTM_User, time=0, value=0}
           {interpMode=RCIM_Cubic, tangentMode=RCTM_User, time=0.35, value=1}
           {arriveTangent=-2.2965, interpMode=RCIM_Cubic, leaveTangent=-2.2965, tangentMode=RCTM_User, time=0.6, value=0.4}
           {interpMode=RCIM_Cubic, tangentMode=RCTM_User, time=1, value=0}
          ]}}}
       Scale Curve=&Multiply_Float(
        A=&RandomRangeFloat(Minimum=1, Maximum=2)
        B=&FloatFromCurve(FloatCurve=DataInterface() {propertyValues {Curve {keys [<at 0 1>,<at 1 0>]}}}, CurveIndex=$Particles.RegionMask.Float)
       )
      )
      B=&Waveform(
       "Waveform [1]"=::NewEnumerator0
       "[1] Phase Scale"=&RandomRangeFloat(Minimum=0.5, Maximum=2)
       "[1] Frequency"=&Multiply_Float(A=&RandomRangeFloat(Minimum=0.5, Maximum=1), B=$System.Main_FlickerSpeed)
       "[1] Amplitude Min/Max"=<0 0.1>
      )
     )
     B=&Multiply_Float(A=1, B=$System.Main_Alpha)
    )
    Scale RGB=&MakeVectorFromLinearColorRGB(Color=&Multiply_LinearColorByFloat(
      Linear Color=&ColorFromCurve(
       ColorCurve=DataInterface() {propertyValues {
         AlphaCurve {keys [<at 0 1>]}
         BlueCurve {keys [<at 0 0.0168 RCIM_Cubic>,<at 0.7144 0.0355>]}
         GreenCurve {keys [<at 0 0.0343 RCIM_Cubic>,<at 0.7144 0.0172>]}
         RedCurve {keys [<at 0 0.147 RCIM_Cubic>,<at 0.7144 0.0522>]}
        }}
       CurveIndex=$Particles.RegionMask.Float
      )
      Float=&RandomRangeFloat(Minimum=25, Maximum=25)
     ))
    ScaleA=true
    ScaleRGB=true
   )
   &DynamicMaterialParameters(
    Index 0 Param 2=1
    Index 0 Param 3=1
    Index 0 Param 4=15
    Module.Index0Float0Write=true
    Module.Index0Float1Write=true
    Module.Index0Float2Write=true
    Module.Index0Float3Write=true
   )
   /Niagara/Modules/Update/Camera/MaintainInCameraParticleScale(Render Target Width=1.5) {module=MaintainInCameraParticleScale001}
   &ScaleSpriteSizeBySpeed(Max Scale Factor=<2 3>)
   &ScaleSpriteSize(Scale Factor=&Vector2DFromFloat(Value=&FloatFromCurve(
      FloatCurve=DataInterface() {propertyValues {Curve {keys [
          {interpMode=RCIM_Cubic, tangentMode=RCTM_User, time=0, value=0.25}
          {interpMode=RCIM_Cubic, tangentMode=RCTM_User, time=0.5, value=1}
          {tangentMode=RCTM_User, time=1, value=1}
         ]}}}
      Scale Curve=&Multiply_Float(
       A=&NormalizeDistanceRange(
        Start Position:NiagaraPosition=$Particles.Position
        End Position:NiagaraPosition=Linked() {linkedVariable {name=System.CameraPostion, type=Vector3f}}
        Distance:NiagaraFloat=1500
        Invert Normalized Range:NiagaraBool=false
       )
       B=&Multiply_Float(A=$Output.ViewMask.ViewMask, B=1)
      )
     )))
   &AlignSpriteToMeshOrientation(SetAlignment=true, SetFacing=true)
   &CameraOffset(Camera Offset Amount=50) {module=CameraOffset001}
   &ScaleColor(Scale Alpha=100, ScaleA=true, ScaleRGB=true) {enabled=false, module=ScaleColor001}
   &ScaleSpriteSize(Scale Sprite Size Mode=::NewEnumerator0, Uniform Scale Factor=3) {enabled=false, module=ScaleSpriteSize001}
  ]
  rendererClasses [&NiagaraSpriteRendererProperties]
  renderers [
   {rendererClass=&NiagaraSpriteRendererProperties}
   {rendererClass=&NiagaraSpriteRendererProperties, rendererIndex=1}
  ]
 }
 Embers_Parent {
  simTarget=GPUComputeSim
  emitterUpdateScript [
   &EmitterState(
    Enable Distance Culling=true
    MaxDistance=18750
    Module.bApplySpawnCountScaleByDistance=true
    Module.bUseMaxDistance=true
    Module.bUseMinDistance=false
    Scalability Mode=::NewEnumerator1
    Scale Spawn Count=true
    Spawn Count Scale By Distance=DataInterface() {propertyValues {Curve {keys [<at 0 0>,<at 0.1 0>,<at 0.5 1>,<at 0.75 1>,<at 1 1>]}}}
   )
   &SpawnRate(SpawnRate=&Multiply_Float(A=375, B=1), Use Spawn Probability=true)
  ]
  particleSpawnScript [
   &InitializeParticle(
    Lifetime Max=3
    Lifetime Min=2
    Lifetime Mode=::NewEnumerator1
    Sprite Size Mode=::NewEnumerator1
    Uniform Sprite Size Max=5
    Uniform Sprite Size Min=3
    UsePositionOffset=false
    Write Material Random=true
   ) {module=InitializeParticle001}
   &CameraOffset(Camera Offset Amount=50)
   &ShapeLocation(MaskSpawnGroup=false, Override Seed=false, Sphere Radius=15000, UseOwnerScale=true)
   /SetVariables(Particles.InitialPos:NiagaraPosition=$Particles.Position) {isSetParameters=true, module=SetVariables_#5}
   /Niagara/Modules/Math/PlaceParticlesOnDepthBuffer_GPU(Max Z Depth=15000, Module.Use Max Z Depth=true)
   &ViewMask(Border Percent Max=25, Border Percent Min=0) {module=ViewMask001}
   &KillParticles(Kill Particles=&SetBoolByFloatComparison(A=1, B=$Output.ViewMask001.ViewMask, Comparison Type=::NewEnumerator6)) {module=KillParticles002}
   &KillParticles(Kill Particles=&SetBoolByFloatComparison(
     A=&VectorLength(Vector=&Subtract_Vector(A=$Particles.InitialPos, B=Linked() {linkedVariable {name=Particles.Position, type=NiagaraPosition}}))
     B=15000
    ))
   &KillParticles(Kill Particles=&Boolean_NotOperation(Bool=$Output.PlaceParticlesOnDepthBuffer_GPU.ValidSample)) {module=KillParticles001}
  ]
  particleUpdateScript [
   &ParticleState()
   &KillParticlesInVolume(Sphere Radius=1000)
   &RegionMask(Region Falloff=1, Region Radius=8000, UseOffset=false, UseRotation=false)
   &PointForce(Force Falloff Distance=250000, Force Strength=-15000, Module.Use Falloff=true, Module.UseLinearForce=false)
   &WindForce(
    Frequency:NiagaraFloat=&RandomRangeFloat(Minimum=0.1, Maximum=2)
    Scale=250
    Wind Speed=<25 0 50>
    Wind Speed Scale=25
   )
   &AerodynamicDrag(
    Aerodynamic Drag:NiagaraFloat=&RandomRangeFloat(Minimum=0.15, Maximum=0.5)
    Lift Contribution=0
    Use Linear Drag=true
    Use Rotational Drag=true
   )
   &SolveForcesAndVelocity(
    Clamp Velocity=true
    Limit Acceleration=false
    Speed Limit:NiagaraFloat=&Multiply_Float(
     A=150
     B=&FloatFromCurve(
      FloatCurve=DataInterface() {propertyValues {Curve {keys [
          <at 0 0.5>
          {arriveTangent=1.255, interpMode=RCIM_Cubic, leaveTangent=1.255, tangentMode=RCTM_User, time=0.5, value=1}
          {arriveTangent=2.5018, interpMode=RCIM_Cubic, leaveTangent=2.5018, tangentMode=RCTM_User, time=1, value=3}
         ]}}}
      CurveIndex=$Particles.RegionMask.Float
     )
    )
   )
   &CalculateAccurateVelocity()
   &ViewMask(Border Percent Max=15)
   &ScaleColor(
    Scale Alpha=&Multiply_Float(
     A=&Multiply_Float(
      A=&FloatFromCurve(
       FloatCurve=DataInterface() {propertyValues {Curve {keys [
           {interpMode=RCIM_Cubic, tangentMode=RCTM_User, time=0, value=0}
           {interpMode=RCIM_Cubic, tangentMode=RCTM_User, time=0.35, value=1}
           {arriveTangent=-2.2965, interpMode=RCIM_Cubic, leaveTangent=-2.2965, tangentMode=RCTM_User, time=0.6, value=0.4}
           {interpMode=RCIM_Cubic, tangentMode=RCTM_User, time=1, value=0}
          ]}}}
       Scale Curve=&Multiply_Float(
        A=&RandomRangeFloat(Minimum=2, Maximum=5)
        B=&FloatFromCurve(FloatCurve=DataInterface() {propertyValues {Curve {keys [<at 0 1>,<at 1 0>]}}}, CurveIndex=$Particles.RegionMask.Float)
       )
      )
      B=&Waveform(
       "Waveform [1]"=::NewEnumerator0
       "[1] Phase Scale"=&Multiply_Float(A=&RandomRangeFloat(Minimum=0.5, Maximum=1), B=$System.Main_FlickerSpeed)
       "[1] Frequency"=&RandomRangeFloat(Minimum=1, Maximum=2)
       "[1] Amplitude Min/Max"=<0 0.1>
      )
     )
     B=&Multiply_Float(A=0.5, B=$System.Main_Alpha)
    )
    Scale RGB=&MakeVectorFromLinearColorRGB(Color=&Multiply_LinearColorByFloat(
      Linear Color=&ColorFromCurve(
       ColorCurve=DataInterface() {propertyValues {
         AlphaCurve {keys [<at 0 1>]}
         BlueCurve {keys [<at 0 0.0168 RCIM_Cubic>,<at 0.7144 0.0355>]}
         GreenCurve {keys [<at 0 0.0343 RCIM_Cubic>,<at 0.7144 0.0172>]}
         RedCurve {keys [<at 0 0.147 RCIM_Cubic>,<at 0.7144 0.0522>]}
        }}
       CurveIndex=$Particles.RegionMask.Float
      )
      Float=&RandomRangeFloat(Minimum=500, Maximum=2500)
     ))
    ScaleA=true
    ScaleRGB=true
   )
   &DynamicMaterialParameters(
    Index 0 Param 2=1
    Index 0 Param 3=1
    Index 0 Param 4=15
    Module.Index0Float0Write=true
    Module.Index0Float1Write=true
    Module.Index0Float2Write=true
    Module.Index0Float3Write=true
   )
   &ScaleSpriteSizeBySpeed(Max Scale Factor=<2 3>)
   &ScaleSpriteSize(Scale Factor=&Vector2DFromFloat(Value=&FloatFromCurve(
      FloatCurve=DataInterface() {propertyValues {Curve {keys [
          {interpMode=RCIM_Cubic, tangentMode=RCTM_User, time=0, value=0.25}
          {interpMode=RCIM_Cubic, tangentMode=RCTM_User, time=0.5, value=1}
          {tangentMode=RCTM_User, time=1, value=1}
         ]}}}
      Scale Curve=&Multiply_Float(
       A=&NormalizeDistanceRange(
        Start Position:NiagaraPosition=$Particles.Position
        End Position:NiagaraPosition=Linked() {linkedVariable {name=System.CameraPostion, type=Vector3f}}
        Distance:NiagaraFloat=1500
        Invert Normalized Range:NiagaraBool=false
       )
       B=&Multiply_Float(A=$Output.ViewMask.ViewMask, B=1)
      )
     )))
   &AlignSpriteToMeshOrientation(SetAlignment=true, SetFacing=true)
   &CameraOffset(Camera Offset Amount=50) {module=CameraOffset001}
   &ScaleSpriteSize(Scale Sprite Size Mode=::NewEnumerator0, Uniform Scale Factor=3) {enabled=false, module=ScaleSpriteSize001}
   &ScaleColor(Scale Alpha=100, ScaleA=true, ScaleRGB=true) {enabled=false, module=ScaleColor001}
  ]
  rendererClasses [&NiagaraSpriteRendererProperties]
  renderers [{rendererClass=&NiagaraSpriteRendererProperties}]
 }
}
summary {
 systemName=NS_PistolCanary_Embers
 emitters [
  {emitterName=Embers_Parent, simTarget=GPUComputeSim, rendererClasses [&NiagaraSpriteRendererProperties]}
  {emitterName=Embers, simTarget=GPUComputeSim, rendererClasses [&NiagaraSpriteRendererProperties]}
 ]
 userVariables []
}
